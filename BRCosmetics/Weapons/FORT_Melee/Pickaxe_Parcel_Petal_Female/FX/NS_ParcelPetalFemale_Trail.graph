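# Niagara system for the weapon trail of the Parcel Petal Female pickaxe, drawn as a short-lived ribbon.
# A single Sticky_Trail emitter uses the AnimTrail spawn, initialize and update modules to build the ribbon from skeletal mesh bones or sockets, nudged by a camera offset.
# A curve tapers ribbon width over life, and that width drives alpha.
# Material texture time scales emitter age by a rate blended on IntegratedDive.
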NiagaraSystem /BRCosmetics/Weapons/FORT_Melee/Pickaxe_Parcel_Petal_Female/FX/NS_ParcelPetalFemale_Trail
dependencies {usedDataInterfaces [&NiagaraDataInterfaceCurve], usedDynamicInputs [], usedModules [&SystemState]}
emitters {Sticky_Trail {
  emitterUpdateScript [
   &EmitterState(Life Cycle Mode=::NewEnumerator1, Loop Behavior=::NewEnumerator1, Loop Duration=0.22, UseLoopDelay=false)
   /Game/Effects/Niagara/Modules/AnimTrail/FN_SpawnAnimTrail/FN_SpawnAnimTrail(Number of Anim Trails=1, SpawnRate=45)
  ]
  particleSpawnScript [
   /Game/Effects/Niagara/Modules/AnimTrail/FN_InitializeAnimTrail/FN_InitializeAnimTrailBoneIndecies(Second Bone or Socket Index=1)
   /Game/Effects/Niagara/Modules/AnimTrail/FN_InitializeAnimTrail/FN_InitializeAnimTrail(AnimTrailWidth=1, Color=<rgba 0.2292 0.2292 0.2292 1>, Lifetime=0.25, Write Material Random=false)
   &CameraOffset(Camera Offset Amount=5)
  ]
  particleUpdateScript [
   &SolveForcesAndVelocity(Clamp Velocity=false, Limit Acceleration=false)
   /Game/Effects/Niagara/Modules/AnimTrail/FN_UpdateAnimTrail/FN_UpdateAnimTrail()
   &ScaleRibbonWidth(Ribbon Width Scale=&FloatFromCurve(FloatCurve=DataInterface() {propertyValues {Curve {keys [<at 0 1 RCIM_Cubic>, {arriveTangent=-1.6223, interpMode=RCIM_Cubic, leaveTangent=-1.6223, tangentMode=RCTM_User, time=1, value=0}]}}}))
   /SetVariables(Particles.RibbonFacing=<1 0 0>) {isSetParameters=true}
   &Material(
    Base Texture Time:NiagaraFloat=&Multiply_Float(A=$Emitter.Age, B=&Lerp_Float(A=-0.4, B=-100, Alpha=$Particles.IntegratedDive))
    PostMult=0.2
    PreMult=0
   )
   &ScaleColor(Scale Alpha=&Multiply_Float(A=0.1, B=$Particles.RibbonWidth), Scale RGB=<5 5 5>, ScaleA=true, ScaleRGB=false)
  ]
  rendererClasses [&NiagaraRibbonRendererProperties]
  renderers [{rendererClass=&NiagaraRibbonRendererProperties}]
 }}
summary {
 systemName=NS_ParcelPetalFemale_Trail
 emitters [{emitterName=Sticky_Trail, rendererClasses [&NiagaraRibbonRendererProperties]}]
 userVariables [
  User.RibbonIDOffset {type=NiagaraInt32}
  User.SkeletalMesh {
   type=NiagaraDataInterfaceSkeletalMesh
   defaultValue {
    struct=NiagaraExt_VariableValue_DataInterface
    value {dataInterface=&self:NiagaraDataInterfaceSkeletalMesh_0}
   }
  }
 ]
}
