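# Control Rig for the Parcel Petal Female pickaxe, adding procedural secondary motion to its plant stems and tongue.
# Wave-only chain harmonics animate the tongue, three tongue-end chains and the bottom stem, with debug drawing on.
# Tongue5 and tongue4 local rotations are then slerped down, smoothed by accumulated lerps and weighted onto main and side stem bones respectively.
# The sideStem1 rotation node is disconnected from execution, so only sideStem2–6 are driven.
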ControlRigBlueprint /BRCosmetics/Weapons/FORT_Melee/Pickaxe_Parcel_Petal_Female/Meshes/Parcel_Petal_Female_CtrlRig/Parcel_Petal_Female_CtrlRig
graphs [
 RigVMFunctionLibrary {links [], nodes []}
 RigVMModel {nodes [
   RigUnit_BeginExecution = &RigUnit_BeginExecution()
   /Script/ControlRig.RigUnit_ChainHarmonics(
    ChainRoot=tongue1
    DrawWorldOffset=(Rotation=(X=0,Y=0,Z=0,W=1),Translation=(X=0,Y=0,Z=0),Scale3D=(X=1,Y=1,Z=1))
    Pendulum=(bEnabled=false,PendulumStiffness=2,PendulumGravity=(X=0,Y=0,Z=0),PendulumBlend=0.75,PendulumDrag=0.98,PendulumMinimum=0,PendulumMaximum=0.1,PendulumEase=Linear,UnwindAxis=(X=0,Y=1,Z=0),UnwindMinimum=0.2,UnwindMaximum=0.05)
    Reach=(bEnabled=false,ReachTarget=(X=10,Y=0,Z=10),ReachAxis=(X=1,Y=0,Z=0),ReachMinimum=0,ReachMaximum=-5,ReachEase=Linear)
    Speed=(X=1,Y=1,Z=1)
    Wave=(bEnabled=True,WaveFrequency=(X=75,Y=75,Z=-75),WaveAmplitude=(X=0.5,Y=1.5,Z=1),WaveOffset=(X=0,Y=1,Z=2),WaveNoise=(X=0,Y=1,Z=1),WaveMinimum=0,WaveMaximum=1,WaveEase=Linear)
    WaveCurve=(EditorCurveData=(Keys=((Value=0.25),(Time=1.0107,Value=0.6084)),DefaultValue=340282346638528859811704183484516925440,PreInfinityExtrap=RCCE_Constant,PostInfinityExtrap=RCCE_Constant))
    bDrawDebug=True
    ExecutePin=@RigUnit_BeginExecution.ExecutePin
   ) {is=struct RigVMUnitNode}
   > &RigUnit_ChainHarmonics(
    ChainRoot=tongueEnd1
    DrawWorldOffset=(Rotation=(X=0,Y=0,Z=0,W=1),Translation=(X=0,Y=0,Z=0),Scale3D=(X=1,Y=1,Z=1))
    Pendulum=(bEnabled=false,PendulumStiffness=2,PendulumGravity=(X=0,Y=0,Z=0),PendulumBlend=0.75,PendulumDrag=0.98,PendulumMinimum=0,PendulumMaximum=0.1,PendulumEase=Linear,UnwindAxis=(X=0,Y=1,Z=0),UnwindMinimum=0.2,UnwindMaximum=0.05)
    Reach=(bEnabled=false,ReachTarget=(X=10,Y=0,Z=10),ReachAxis=(X=1,Y=0,Z=0),ReachMinimum=0,ReachMaximum=-5,ReachEase=Linear)
    Speed=(X=1,Y=1,Z=1)
    Wave=(bEnabled=True,WaveFrequency=(X=75,Y=75,Z=-75),WaveAmplitude=(X=0.5,Y=1.5,Z=1),WaveOffset=(X=0,Y=1,Z=2),WaveNoise=(X=0,Y=1,Z=1),WaveMinimum=0,WaveMaximum=1,WaveEase=Linear)
    WaveCurve=(EditorCurveData=(Keys=((Value=0.25),(Time=1.0107,Value=0.6084)),DefaultValue=340282346638528859811704183484516925440,PreInfinityExtrap=RCCE_Constant,PostInfinityExtrap=RCCE_Constant))
    bDrawDebug=True
   )
   > &RigUnit_ChainHarmonics(
    ChainRoot=tongueEnd2
    DrawWorldOffset=(Rotation=(X=0,Y=0,Z=0,W=1),Translation=(X=0,Y=0,Z=0),Scale3D=(X=1,Y=1,Z=1))
    Pendulum=(bEnabled=false,PendulumStiffness=2,PendulumGravity=(X=0,Y=0,Z=0),PendulumBlend=0.75,PendulumDrag=0.98,PendulumMinimum=0,PendulumMaximum=0.1,PendulumEase=Linear,UnwindAxis=(X=0,Y=1,Z=0),UnwindMinimum=0.2,UnwindMaximum=0.05)
    Reach=(bEnabled=false,ReachTarget=(X=10,Y=0,Z=10),ReachAxis=(X=1,Y=0,Z=0),ReachMinimum=0,ReachMaximum=-5,ReachEase=Linear)
    Speed=(X=1,Y=1,Z=1)
    Wave=(bEnabled=True,WaveFrequency=(X=75,Y=75,Z=-75),WaveAmplitude=(X=0.5,Y=1.5,Z=1),WaveOffset=(X=0,Y=1,Z=2),WaveNoise=(X=0,Y=1,Z=1),WaveMinimum=0,WaveMaximum=1,WaveEase=Linear)
    WaveCurve=(EditorCurveData=(Keys=((Value=0.25),(Time=1.0107,Value=0.6084)),DefaultValue=340282346638528859811704183484516925440,PreInfinityExtrap=RCCE_Constant,PostInfinityExtrap=RCCE_Constant))
    bDrawDebug=True
   )
   > &RigUnit_ChainHarmonics(
    ChainRoot=tongueEnd3
    DrawWorldOffset=(Rotation=(X=0,Y=0,Z=0,W=1),Translation=(X=0,Y=0,Z=0),Scale3D=(X=1,Y=1,Z=1))
    Pendulum=(bEnabled=false,PendulumStiffness=2,PendulumGravity=(X=0,Y=0,Z=0),PendulumBlend=0.75,PendulumDrag=0.98,PendulumMinimum=0,PendulumMaximum=0.1,PendulumEase=Linear,UnwindAxis=(X=0,Y=1,Z=0),UnwindMinimum=0.2,UnwindMaximum=0.05)
    Reach=(bEnabled=false,ReachTarget=(X=10,Y=0,Z=10),ReachAxis=(X=1,Y=0,Z=0),ReachMinimum=0,ReachMaximum=-5,ReachEase=Linear)
    Speed=(X=1,Y=1,Z=1)
    Wave=(bEnabled=True,WaveFrequency=(X=75,Y=75,Z=-75),WaveAmplitude=(X=0.5,Y=1.5,Z=1),WaveOffset=(X=0,Y=1,Z=2),WaveNoise=(X=0,Y=1,Z=1),WaveMinimum=0,WaveMaximum=1,WaveEase=Linear)
    WaveCurve=(EditorCurveData=(Keys=((Value=0.25),(Time=1.0107,Value=0.6084)),DefaultValue=340282346638528859811704183484516925440,PreInfinityExtrap=RCCE_Constant,PostInfinityExtrap=RCCE_Constant))
    bDrawDebug=True
   )
   > &RigUnit_ChainHarmonics(
    ChainRoot=bottomStem1
    DrawWorldOffset=(Rotation=(X=0,Y=0,Z=0,W=1),Translation=(X=0,Y=0,Z=0),Scale3D=(X=1,Y=1,Z=1))
    Pendulum=(bEnabled=false,PendulumStiffness=2,PendulumGravity=(X=0,Y=0,Z=0),PendulumBlend=0.75,PendulumDrag=0.98,PendulumMinimum=0,PendulumMaximum=0.1,PendulumEase=Linear,UnwindAxis=(X=0,Y=1,Z=0),UnwindMinimum=0.2,UnwindMaximum=0.05)
    Reach=(bEnabled=false,ReachTarget=(X=10,Y=0,Z=10),ReachAxis=(X=1,Y=0,Z=0),ReachMinimum=0,ReachMaximum=-5,ReachEase=Linear)
    Speed=(X=1,Y=1,Z=1)
    Wave=(bEnabled=True,WaveFrequency=(X=50,Y=50,Z=-50),WaveAmplitude=(X=0.25,Y=0.75,Z=0.5),WaveOffset=(X=0,Y=0.5,Z=1),WaveNoise=(X=0,Y=1,Z=1),WaveMinimum=0,WaveMaximum=1,WaveEase=Linear)
    WaveCurve=(EditorCurveData=(Keys=((Value=0.25),(Time=1.0107,Value=0.6084)),DefaultValue=340282346638528859811704183484516925440,PreInfinityExtrap=RCCE_Constant,PostInfinityExtrap=RCCE_Constant))
    bDrawDebug=True
   )
   n1 = &RigUnit_GetBoneTransform(Bone=tongue5, Space=LocalSpace)
   > /Script/ControlRig.RigUnit_SetBoneRotation(
    Bone=mainStem1
    Space=LocalSpace
    Weight=0.1
    bPropagateToChildren=True
    Rotation=/Script/RigVM.RigVMFunction_AccumulateQuatLerp(
     Blend=0.5
     InitialValue=(X=0,Y=0,Z=0,W=1)
     bIntegrateDeltaTime=false
     TargetValue=/Script/RigVM.RigVMFunction_MathQuaternionSlerp(A=(X=0,Y=0,Z=0,W=1), T=0.4, B=@n1.Transform.Rotation) {is=struct RigVMUnitNode, template="Interpolate::Execute(in A,in B,in T,out Result)"}.Result
    ) {is=struct RigVMUnitNode, template="AccumulateLerp::Execute(in TargetValue,in InitialValue,in Blend,in bIntegrateDeltaTime,out Result)"}.Result
   ) {is=struct RigVMUnitNode}
   > &RigUnit_SetBoneRotation(
    Bone=mainStem2
    Weight=0.1
    bPropagateToChildren=True
    Rotation=&RigVMFunction_AccumulateQuatLerp(
     Blend=0.3
     InitialValue=(X=0,Y=0,Z=0,W=1)
     TargetValue=&RigVMFunction_MathQuaternionSlerp(A=(X=0,Y=0,Z=0,W=1), T=0.2, B=@n1.Transform.Rotation) {template="Interpolate::Execute(in A,in B,in T,out Result)"}.Result
    ) {template="AccumulateLerp::Execute(in TargetValue,in InitialValue,in Blend,in bIntegrateDeltaTime,out Result)"}.Result
   )
   > &RigUnit_SetBoneRotation(
    Bone=mainStem3
    Weight=0.1
    bPropagateToChildren=True
    Rotation=&RigVMFunction_AccumulateQuatLerp(
     Blend=0.2
     InitialValue=(X=0,Y=0,Z=0,W=1)
     TargetValue=&RigVMFunction_MathQuaternionSlerp(A=(X=0,Y=0,Z=0,W=1), T=0.3, B=@n1.Transform.Rotation) {template="Interpolate::Execute(in A,in B,in T,out Result)"}.Result
    ) {template="AccumulateLerp::Execute(in TargetValue,in InitialValue,in Blend,in bIntegrateDeltaTime,out Result)"}.Result
   )
   > &RigUnit_SetBoneRotation(
    Bone=mainStem4
    Weight=0.1
    bPropagateToChildren=True
    Rotation=&RigVMFunction_AccumulateQuatLerp(
     Blend=0.3
     InitialValue=(X=0,Y=0,Z=0,W=1)
     TargetValue=&RigVMFunction_MathQuaternionSlerp(A=(X=0,Y=0,Z=0,W=1), T=0.4, B=@n1.Transform.Rotation) {template="Interpolate::Execute(in A,in B,in T,out Result)"}.Result
    ) {template="AccumulateLerp::Execute(in TargetValue,in InitialValue,in Blend,in bIntegrateDeltaTime,out Result)"}.Result
   )
   > &RigUnit_SetBoneRotation(
    Bone=mainStem5
    Weight=0.1
    bPropagateToChildren=True
    Rotation=&RigVMFunction_AccumulateQuatLerp(
     Blend=0.4
     InitialValue=(X=0,Y=0,Z=0,W=1)
     TargetValue=&RigVMFunction_MathQuaternionSlerp(A=(X=0,Y=0,Z=0,W=1), T=0.5, B=@n1.Transform.Rotation) {template="Interpolate::Execute(in A,in B,in T,out Result)"}.Result
    ) {template="AccumulateLerp::Execute(in TargetValue,in InitialValue,in Blend,in bIntegrateDeltaTime,out Result)"}.Result
   )
   > &RigUnit_SetBoneRotation(
    Bone=mainStem6
    Weight=0.1
    bPropagateToChildren=True
    Rotation=&RigVMFunction_AccumulateQuatLerp(
     Blend=0.2
     InitialValue=(X=0,Y=0,Z=0,W=1)
     TargetValue=&RigVMFunction_MathQuaternionSlerp(A=(X=0,Y=0,Z=0,W=1), T=0.3, B=@n1.Transform.Rotation) {template="Interpolate::Execute(in A,in B,in T,out Result)"}.Result
    ) {template="AccumulateLerp::Execute(in TargetValue,in InitialValue,in Blend,in bIntegrateDeltaTime,out Result)"}.Result
   )
   n2 = &RigUnit_GetBoneTransform(Bone=tongue4, Space=LocalSpace)
   > &RigUnit_SetBoneRotation(
    Bone=sideStem2
    Weight=0.5
    bPropagateToChildren=True
    Rotation=&RigVMFunction_AccumulateQuatLerp(
     Blend=0.3
     InitialValue=(X=0,Y=0,Z=0,W=1)
     TargetValue=&RigVMFunction_MathQuaternionSlerp(A=(X=0,Y=0,Z=0,W=1), T=0.3, B=@n2.Transform.Rotation) {template="Interpolate::Execute(in A,in B,in T,out Result)"}.Result
    ) {template="AccumulateLerp::Execute(in TargetValue,in InitialValue,in Blend,in bIntegrateDeltaTime,out Result)"}.Result
   )
   > &RigUnit_SetBoneRotation(
    Bone=sideStem3
    Weight=0.1
    bPropagateToChildren=True
    Rotation=&RigVMFunction_AccumulateQuatLerp(
     Blend=0.2
     InitialValue=(X=0,Y=0,Z=0,W=1)
     TargetValue=&RigVMFunction_MathQuaternionSlerp(A=(X=0,Y=0,Z=0,W=1), T=0.4, B=@n2.Transform.Rotation) {template="Interpolate::Execute(in A,in B,in T,out Result)"}.Result
    ) {template="AccumulateLerp::Execute(in TargetValue,in InitialValue,in Blend,in bIntegrateDeltaTime,out Result)"}.Result
   )
   > &RigUnit_SetBoneRotation(
    Bone=sideStem4
    Weight=0.1
    bPropagateToChildren=True
    Rotation=&RigVMFunction_AccumulateQuatLerp(
     Blend=0.3
     InitialValue=(X=0,Y=0,Z=0,W=1)
     TargetValue=&RigVMFunction_MathQuaternionSlerp(A=(X=0,Y=0,Z=0,W=1), T=0.2, B=@n2.Transform.Rotation) {template="Interpolate::Execute(in A,in B,in T,out Result)"}.Result
    ) {template="AccumulateLerp::Execute(in TargetValue,in InitialValue,in Blend,in bIntegrateDeltaTime,out Result)"}.Result
   )
   > &RigUnit_SetBoneRotation(
    Bone=sideStem5
    Weight=0.1
    bPropagateToChildren=True
    Rotation=&RigVMFunction_AccumulateQuatLerp(
     Blend=0.4
     InitialValue=(X=0,Y=0,Z=0,W=1)
     TargetValue=&RigVMFunction_MathQuaternionSlerp(A=(X=0,Y=0,Z=0,W=1), T=0.3, B=@n2.Transform.Rotation) {template="Interpolate::Execute(in A,in B,in T,out Result)"}.Result
    ) {template="AccumulateLerp::Execute(in TargetValue,in InitialValue,in Blend,in bIntegrateDeltaTime,out Result)"}.Result
   )
   > &RigUnit_SetBoneRotation(
    Bone=sideStem6
    Weight=0.1
    bPropagateToChildren=True
    Rotation=&RigVMFunction_AccumulateQuatLerp(
     Blend=0.2
     InitialValue=(X=0,Y=0,Z=0,W=1)
     TargetValue=&RigVMFunction_MathQuaternionSlerp(A=(X=0,Y=0,Z=0,W=1), T=0.4, B=@n2.Transform.Rotation) {template="Interpolate::Execute(in A,in B,in T,out Result)"}.Result
    ) {template="AccumulateLerp::Execute(in TargetValue,in InitialValue,in Blend,in bIntegrateDeltaTime,out Result)"}.Result
   )
   &RigUnit_SetBoneRotation(
    Bone=sideStem1
    ExecutePin=()
    Space=LocalSpace
    Weight=0.1
    bPropagateToChildren=True
    Rotation=&RigVMFunction_AccumulateQuatLerp(
     Blend=0.5
     InitialValue=(X=0,Y=0,Z=0,W=1)
     bIntegrateDeltaTime=false
     TargetValue=&RigVMFunction_MathQuaternionSlerp(A=(X=0,Y=0,Z=0,W=1), T=0.5, B=@n2.Transform.Rotation) {template="Interpolate::Execute(in A,in B,in T,out Result)"}.Result
    ) {template="AccumulateLerp::Execute(in TargetValue,in InitialValue,in Blend,in bIntegrateDeltaTime,out Result)"}.Result
   )
  ]}
]
hierarchy [
 b_AR_Root {loc=<0 0 0>, rot=<0 0 0.7071 0.7071>}
 Melee_TwoHanded_Handle {parent=b_AR_Root, loc=<8.1477 -0.008 -29.0793>, rot=<0 0 0.7071 -0.7071>}
 sideStem1 {parent=Melee_TwoHanded_Handle, loc=<1.3091 -6.255 53.3034>, rot=<0.6592 -0.6289 0.4008 0.0965>}
 sideStem2 {parent=sideStem1, loc=<4.417 0 0>, rot=<-0.0369 0.1575 0.0015 -0.9868>}
 sideStem3 {parent=sideStem2, loc=<6.5381 0 0>, rot=<0.0508 0.0947 0.0336 -0.9936>}
 sideStem4 {parent=sideStem3, loc=<6.5516 0 0>, rot=<-0.082 -0.448 -0.0448 -0.8892>}
 sideStem5 {parent=sideStem4, loc=<4.8953 0 0>, rot=<-0.0628 -0.2684 0.0215 -0.961>}
 sideStem6 {parent=sideStem5, loc=<3.424 0 0>, rot=<-0.0415 -0.3095 -0.0035 -0.95>}
 mouthUpper1 {parent=sideStem6, loc=<3.5648 -0.0106 -0.1264>, rot=<0.9511 -0.0481 -0.2687 -0.1444>}
 mouthUpper2 {parent=mouthUpper1, loc=<15.9187 0 0>, rot=<0 0 0 -1>}
 mouthLower1 {parent=sideStem6, loc=<3.5304 -0.1539 -0.9192>, rot=<0.9468 -0.0453 -0.2775 -0.1567>}
 mouthLower2 {parent=mouthLower1, loc=<15.4488 0 0>, rot=<0 0 0 -1>}
 leafLower1_r {parent=sideStem3, loc=<6.3672 1.7905 -2.3964>, rot=<0.8033 0.2555 -0.1408 -0.5192>}
 leafLower2_r {parent=leafLower1_r, loc=<3.1675 0 0>, rot=<-0.31 -0.3311 0.0074 -0.8912>}
 leafLower3_r {parent=leafLower2_r, loc=<4.3251 0 0>, rot=<0.1763 0 0 -0.9843>}
 leafLower1_l {parent=sideStem3, loc=<8.8729 -1.5706 -0.5678>, rot=<0.7979 -0.2894 -0.263 -0.4587>}
 leafLower2_l {parent=leafLower1_l, loc=<2.307 0 0>, rot=<0.1003 -0.0976 -0.1484 -0.979>}
 leafLower3_l {parent=leafLower2_l, loc=<3.4962 0 0>, rot=<0 0 0 1>}
 mainStem1 {parent=Melee_TwoHanded_Handle, loc=<3.967 -14.6294 69.5343>, rot=<0.5912 -0.5575 0.3999 0.4241>}
 mainStem2 {parent=mainStem1, loc=<6.4479 0 0>, rot=<0.002 -0.1866 -0.0474 -0.9813>}
 mainStem3 {parent=mainStem2, loc=<8.5992 0 0>, rot=<-0.0163 -0.2696 -0.0216 -0.9626>}
 mainStem4 {parent=mainStem3, loc=<6.3697 0 0>, rot=<-0.0417 -0.3451 -0.0352 -0.937>}
 mainStem5 {parent=mainStem4, loc=<7.0849 0 0>, rot=<-0.0361 -0.2536 0.0158 -0.9665>}
 mainStem6 {parent=mainStem5, loc=<6.5923 0 0>, rot=<0 0 0 1>}
 head {parent=mainStem6, loc=<1.4136 1.7852 -0.1017>, rot=<0.0079 0.1263 0.0619 -0.99>}
 mouthUp1 {parent=head, loc=<3.4557 -2.5296 -1.0787>, rot=<0.0033 0.3315 0.0093 -0.9434>}
 mouthUp2 {parent=mouthUp1, loc=<13.0211 0 0>, rot=<-0.0014 -0.4388 -0.006 -0.8986>}
 mouthUp3 {parent=mouthUp2, loc=<12.0672 0 0>, rot=<0 0.4975 0 -0.8675>}
 mouthLeft1 {parent=head, loc=<2.6923 -4.4903 -2.7956>, rot=<0.6331 -0.1053 -0.6147 0.4585>}
 mouthLeft2 {parent=mouthLeft1, loc=<12.8249 0 0>, rot=<-0.0816 -0.4595 0.0038 -0.8844>}
 mouthLeft3 {parent=mouthLeft2, loc=<12.5413 0 0>, rot=<0 0.453 0 -0.8915>}
 mouthRight1 {parent=head, loc=<3.2479 -0.6271 -3.3039>, rot=<0.6808 0.1218 -0.5757 -0.4361>}
 mouthRight2 {parent=mouthRight1, loc=<13.6806 -0.0667 0.5488>, rot=<-0.0411 -0.4649 -0.0221 -0.8842>}
 mouthRight3 {parent=mouthRight2, loc=<9.6735 0 0>, rot=<0 0.4687 0 -0.8834>}
 tongue1 {parent=head, loc=<3.3427 -2.8396 -2.8684>, rot=<0.0058 -0.1501 -0.038 -0.9879>}
 tongue2 {parent=tongue1, loc=<10.6113 0 0>, rot=<-0.0072 0.0575 0.0289 -0.9979>}
 tongue3 {parent=tongue2, loc=<6.3365 0 0>, rot=<-0.0056 0.0004 0.0298 -0.9995>}
 tongue4 {parent=tongue3, loc=<6.8283 0 0>, rot=<0.0068 0.0075 -0.0376 -0.9992>}
 tongue5 {parent=tongue4, loc=<6.9646 0 0>, rot=<-0.0227 -0.018 0.1186 -0.9925>}
 tongue6 {parent=tongue5, loc=<2.77 0 0>, rot=<0.1261 0.3205 0.1459 -0.9274>}
 tongueEnd1 {parent=tongue6, loc=<1.8444 -0.8345 0>, rot=<0 0 0.484 -0.875>}
 tongueEnd2 {parent=tongue6, loc=<2.0306 0.4579 0.4761>, rot=<-0.1382 0.5027 0.0253 -0.8529>}
 NewBone {parent=tongueEnd2, loc=<0 0 0>, rot=<0 0 0 1>}
 tongueEnd3 {parent=tongue5, loc=<3.4919 0.6224 -1.1902>, rot=<0.2858 -0.7056 0.0181 -0.6482>}
 tongueEnd4 {parent=tongueEnd3, loc=<1.9549 0 0>, rot=<-0.0028 -0.403 -0.2695 -0.8746>}
 leafUpper1_r {parent=mainStem5, loc=<1.8506 3.0851 -3.6302>, rot=<-0.4372 -0.7286 -0.5058 0.1491>}
 leafUpper2_r {parent=leafUpper1_r, loc=<4.2727 0 0>, rot=<0.0649 -0.1442 -0.0364 -0.9867>}
 leafUpper3_r {parent=leafUpper2_r, loc=<6.2346 0 0>, rot=<0 0 0 -1>}
 leafUpper1_l {parent=mainStem5, loc=<0.5724 -4.5499 -1.3813>, rot=<0.2328 0.5073 0.7606 -0.3315>}
 leafUpper2_l {parent=leafUpper1_l, loc=<3.2483 0 0>, rot=<-0.0056 0.0367 -0.2981 -0.9538>}
 leafUpper3_l {parent=leafUpper2_l, loc=<5.3241 0 0>, rot=<0 0 0 -1>}
 bottomStem1 {parent=Melee_TwoHanded_Handle, loc=<0.5574 -3.651 -28.1061>, rot=<0.5051 -0.4254 -0.4837 -0.5743>}
 bottomStem2 {parent=bottomStem1, loc=<5.9905 0 0>, rot=<0.0008 -0.0074 -0.0674 -0.9977>}
 bottomStem3 {parent=bottomStem2, loc=<4.9752 0 0>, rot=<0.0527 -0.3422 0.1093 -0.9318>}
 bottomStem4 {parent=bottomStem3, loc=<4.2223 0 0>, rot=<0.1594 -0.7115 -0.0336 -0.6836>}
 bottomStem5 {parent=bottomStem4, loc=<3.6849 0 0>, rot=<0 0 0 1>}
]
