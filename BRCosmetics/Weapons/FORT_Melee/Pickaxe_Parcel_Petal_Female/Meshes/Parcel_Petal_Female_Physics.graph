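# Physics asset for the Parcel Petal Female two-handed pickaxe, a plant-like weapon with leaves, a three-part mouth and a branching tongue.
# Four three-bone leaf chains hang off the box-shaped handle body, with the lower pair held stiffer than the upper.
# The capsule head swings and twists freely on the handle and anchors three mouth chains and the tongue chain.
# These chains use small angular limits for subtle secondary motion.
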PhysicsAsset /BRCosmetics/Weapons/FORT_Melee/Pickaxe_Parcel_Petal_Female/Meshes/Parcel_Petal_Female_Physics
bodies {
 Melee_TwoHanded_Handle [{shapeType=Box, extentX=0.5, extentY=0.5, extentZ=0.5, center=<1.831 3.2746 30.0653>}]
 head [{shapeType=Capsule, length=12.75, radius=4.25, center=<-9.8778 -3.0038 1.5683>, rotation=<rot -90 0.0198 -0.0199>}]
 leafLower1_l [{shapeType=Box, extentX=1.01, extentY=1.01, extentZ=1.01, center=<1.172 -0.1112 0.0225>}]
 leafLower1_r [{shapeType=Box, extentX=1.01, extentY=1.01, extentZ=1.01, center=<1.119 -0.072 -0.166>}]
 leafLower2_l [{shapeType=Box, extentX=5, extentY=8, extentZ=1, center=<2.5 0 0>}]
 leafLower2_r [{shapeType=Box, extentX=5, extentY=8, extentZ=1, center=<2.5 0 0>}]
 leafLower3_l [{shapeType=Box, extentX=4, extentY=8, extentZ=1, center=<2 0 0>}]
 leafLower3_r [{shapeType=Box, extentX=4, extentY=8, extentZ=1, center=<2 0 0>}]
 leafUpper1_l [{shapeType=Box, extentX=3.1349, extentY=1.1756, extentZ=0.8439, center=<1.7587 -0.1806 0.012>}]
 leafUpper1_r [{shapeType=Box, extentX=4.0157, extentY=0.8208, extentZ=0.8493, center=<2.2013 -0.0957 0.019>}]
 leafUpper2_l [{shapeType=Box, extentX=5, extentY=8, extentZ=1, center=<2.5 0 0>}]
 leafUpper2_r [{shapeType=Box, extentX=5, extentY=8, extentZ=1, center=<2.5 0 0>}]
 leafUpper3_l [{shapeType=Box, extentX=4, extentY=8, extentZ=1, center=<2 0 0>}]
 leafUpper3_r [{shapeType=Box, extentX=4, extentY=8, extentZ=1, center=<2 0 0>}]
 mouthLeft1 [{shapeType=Box, extentX=15.0431, extentY=17.2444, extentZ=10.2106, center=<6.5576 0.1942 -2.5884>}]
 mouthLeft2 [{shapeType=Box, extentX=11.5194, extentY=15.2421, extentZ=10.1995, center=<8.417 -0.7268 -4.3835>}]
 mouthLeft3 [{shapeType=Box, extentX=5.7957, extentY=4.6182, extentZ=2.0232, center=<3.2293 0.1773 -0.9714>}]
 mouthRight1 [{shapeType=Box, extentX=16.1417, extentY=16.8857, extentZ=11.9399, center=<6.8803 0.4586 -2.8845>}]
 mouthRight2 [{shapeType=Box, extentX=8.2526, extentY=14.2347, extentZ=9.0606, center=<8.0991 0.0518 -3.1429>}]
 mouthRight3 [{shapeType=Box, extentX=6.9618, extentY=6.4455, extentZ=2.7219, center=<4.4837 -0.0984 -1.4711>}]
 mouthUp1 [{shapeType=Box, extentX=16.0255, extentY=17.0123, extentZ=11.2027, center=<7.2238 -0.2561 -2.5099>}]
 mouthUp2 [{shapeType=Box, extentX=9.5091, extentY=15.1345, extentZ=8.4415, center=<8.1797 -0.1538 -3.9274>}]
 mouthUp3 [{shapeType=Box, extentX=6.2763, extentY=5.7969, extentZ=2.0524, center=<2.6198 -0.1734 -0.5559>}]
 tongue1 [{shapeType=Box, extentX=6.7893, extentY=5.6096, extentZ=4.8521, center=<3.2582 0.1317 -0.0713>}]
 tongue2 [{shapeType=Box, extentX=8.6625, extentY=4.2666, extentZ=4.1153, center=<0.1334 -0.2657 0.1028>}]
 tongue3 [{shapeType=Box, extentX=5.7324, extentY=3.1956, extentZ=3.4309, center=<0.9535 -0.15 -0.2202>}]
 tongue4 [{shapeType=Box, extentX=7.5845, extentY=3.2351, extentZ=2.9202, center=<0.4037 0.0102 -0.0566>}]
 tongue5 [{shapeType=Box, extentX=5.4965, extentY=2.4036, extentZ=1.9723, center=<0.3264 0.0145 -0.2547>}]
 tongue6 [{shapeType=Box, extentX=1.6696, extentY=1.638, extentZ=1.309, center=<0.6729 0.1974 -0.1471>}]
 tongueEnd1 [{shapeType=Box, extentX=1.8864, extentY=1.1587, extentZ=1.3853, center=<0.2887 -0.2487 0.1562>}]
 tongueEnd2 [{shapeType=Box, extentX=2.7553, extentY=0.8508, extentZ=0.7785, center=<0.5901 0.036 0.0415>}]
 tongueEnd3 [{shapeType=Box, extentX=1.3293, extentY=1.5419, extentZ=1.3501, center=<0.33 -0.2072 0.3578>}]
 tongueEnd4 [{shapeType=Box, extentX=2.5422, extentY=1.3557, extentZ=0.826, center=<0.2998 0.1402 -0.1011>}]
}
constraints [
 {bone1Name=leafUpper1_l, bone2Name=Melee_TwoHanded_Handle, swing1LimitDegrees=10, swing2LimitDegrees=10, twistLimitDegrees=3}
 {bone1Name=leafUpper2_l, bone2Name=leafUpper1_l, swing1LimitDegrees=10, swing2LimitDegrees=10, twistLimitDegrees=3}
 {bone1Name=leafUpper3_l, bone2Name=leafUpper2_l, swing1LimitDegrees=10, swing2LimitDegrees=10, twistLimitDegrees=3}
 {bone1Name=leafUpper1_r, bone2Name=Melee_TwoHanded_Handle, swing1LimitDegrees=10, swing2LimitDegrees=10, twistLimitDegrees=3}
 {bone1Name=leafUpper2_r, bone2Name=leafUpper1_r, swing1LimitDegrees=10, swing2LimitDegrees=10, twistLimitDegrees=3}
 {bone1Name=leafUpper3_r, bone2Name=leafUpper2_r, swing1LimitDegrees=10, swing2LimitDegrees=10, twistLimitDegrees=3}
 {bone1Name=leafLower1_l, bone2Name=Melee_TwoHanded_Handle, swing1LimitDegrees=3, swing2LimitDegrees=3, twistLimitDegrees=1}
 {bone1Name=leafLower2_l, bone2Name=leafLower1_l, swing1LimitDegrees=3, swing2LimitDegrees=3, twistLimitDegrees=1}
 {bone1Name=leafLower3_l, bone2Name=leafLower2_l, swing1LimitDegrees=3, swing2LimitDegrees=3, twistLimitDegrees=1}
 {bone1Name=leafLower1_r, bone2Name=Melee_TwoHanded_Handle, swing1LimitDegrees=3, swing2LimitDegrees=3, twistLimitDegrees=1}
 {bone1Name=leafLower2_r, bone2Name=leafLower1_r, swing1LimitDegrees=3, swing2LimitDegrees=3, twistLimitDegrees=1}
 {bone1Name=leafLower3_r, bone2Name=leafLower2_r, swing1LimitDegrees=3, swing2LimitDegrees=3, twistLimitDegrees=1}
 {
  bone1Name=head
  bone2Name=Melee_TwoHanded_Handle
  swing1LimitDegrees=45
  swing1Motion=Free
  swing2LimitDegrees=45
  swing2Motion=Free
  twistLimitDegrees=45
  twistMotion=Free
 }
 {bone1Name=mouthUp1, bone2Name=head, swing1LimitDegrees=5, swing2LimitDegrees=5, twistLimitDegrees=2}
 {bone1Name=mouthUp3, bone2Name=mouthUp2, swing1LimitDegrees=10, swing2LimitDegrees=10, twistLimitDegrees=2}
 {bone1Name=mouthLeft1, bone2Name=head, swing1LimitDegrees=5, swing2LimitDegrees=5, twistLimitDegrees=2}
 {bone1Name=mouthLeft3, bone2Name=mouthLeft2, swing1LimitDegrees=10, swing2LimitDegrees=10, twistLimitDegrees=2}
 {bone1Name=mouthRight1, bone2Name=head, swing1LimitDegrees=5, swing2LimitDegrees=5, twistLimitDegrees=2}
 {bone1Name=mouthRight3, bone2Name=mouthRight2, swing1LimitDegrees=10, swing2LimitDegrees=10, twistLimitDegrees=2}
 {bone1Name=tongue1, bone2Name=head, swing1LimitDegrees=4, swing2LimitDegrees=4, twistLimitDegrees=2}
 {bone1Name=tongue2, bone2Name=tongue1, swing1LimitDegrees=4, swing2LimitDegrees=4, twistLimitDegrees=2}
 {bone1Name=tongue3, bone2Name=tongue2, swing1LimitDegrees=4, swing2LimitDegrees=4, twistLimitDegrees=2}
 {bone1Name=tongue4, bone2Name=tongue3, swing1LimitDegrees=4, swing2LimitDegrees=4, twistLimitDegrees=2}
 {bone1Name=tongue5, bone2Name=tongue4, swing1LimitDegrees=4, swing2LimitDegrees=4, twistLimitDegrees=2}
 {bone1Name=tongue6, bone2Name=tongue5, swing1LimitDegrees=4, swing2LimitDegrees=4, twistLimitDegrees=2}
 {bone1Name=tongueEnd1, bone2Name=tongue6, swing1LimitDegrees=5, swing2LimitDegrees=5, twistLimitDegrees=2}
 {bone1Name=tongueEnd2, bone2Name=tongue6, swing1LimitDegrees=5, swing2LimitDegrees=5, twistLimitDegrees=2}
 {bone1Name=tongueEnd3, bone2Name=tongue5, swing1LimitDegrees=5, swing2LimitDegrees=5, twistLimitDegrees=2}
 {bone1Name=tongueEnd4, bone2Name=tongueEnd3, swing1LimitDegrees=5, swing2LimitDegrees=5, twistLimitDegrees=2}
 {bone1Name=mouthUp2, bone2Name=mouthUp1, swing1LimitDegrees=5, swing2LimitDegrees=5, twistLimitDegrees=2}
 {bone1Name=mouthLeft2, bone2Name=mouthLeft1, swing1LimitDegrees=5, swing2LimitDegrees=5, twistLimitDegrees=2}
 {bone1Name=mouthRight2, bone2Name=mouthRight1, swing1LimitDegrees=5, swing2LimitDegrees=5, twistLimitDegrees=2}
]
