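# Animation Blueprint for the Parcel Petal Female pickaxe cosmetic, adding secondary physics and an idle mouth wobble.
# The incoming pose runs through a RigidBody node simulated relative to the two-handed handle bone, with overridden gravity and damped component motion.
# Additive bone rotations then open and close the upper and lower mouth bones, driven by sine-of-game-time pitch values set each update.
# A Control Rig finishes the pose before output.
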AnimBlueprint /BRCosmetics/Weapons/FORT_Melee/Pickaxe_Parcel_Petal_Female/Meshes/Parcel_Petal_Female_AnimBP
target_skeleton=&/Parcel_Petal_Female_Skeleton
graphs [
 AnimGraph {nodes [
   AnimGraphNode_SubInput_0 = Input Pose()
   n1 = RigidBody(ComponentPose=Local To Component(LocalPose=@AnimGraphNode_SubInput_0)) {
    bClampLinearTranslationLimitToRefPose=true
    bOverrideWorldGravity=true
    overridePhysicsAsset=&/Parcel_Petal_Female_Physics
    simulationSpace=BaseBoneSpace
    alphaBoolBlend {blendOption=Linear}
    baseBoneRef {boneName=Melee_TwoHanded_Handle}
    componentAppliedLinearAccClamp=<5000 5000 5000>
    componentLinearAccScale=<0.4 0.4 0.4>
    componentLinearVelScale=<0.3 0.3 0.3>
    overrideWorldGravity {z=-980}
   }
   n2 = "Transform (Modify) Bone - Bone: mouthUpper1"(ComponentPose=@n1, Alpha=1, Translation="0,0,0", Rotation=Get mouthUpperRot() -> (mouthUpperRot:Rotator), Scale="1,1,1") {
    rotationMode=BMM_Additive
    rotationSpace=BCS_BoneSpace
    translationMode=BMM_Additive
    translationSpace=BCS_BoneSpace
    alphaBoolBlend {blendOption=Linear}
    boneToModify {boneName=mouthUpper1}
   }
   n3 = "Transform (Modify) Bone - Bone: mouthLower1"(ComponentPose=@n2, Alpha=1, Translation="0,0,0", Rotation=Get mouthLowerRot() -> (mouthLowerRot:Rotator), Scale="1,1,1") {
    rotationMode=BMM_Additive
    rotationSpace=BCS_BoneSpace
    translationMode=BMM_Additive
    translationSpace=BCS_BoneSpace
    alphaBoolBlend {blendOption=Linear}
    boneToModify {boneName=mouthLower1}
   }
   n4 = Control Rig(Source=Component To Local(ComponentPose=@n3)) {
    alphaBoolBlend {blendOption=Linear}
    controlRigAssetReference {blueprintRigClass=/BRCosmetics/Weapons/FORT_Melee/Pickaxe_Parcel_Petal_Female/Meshes/Parcel_Petal_Female_CtrlRig/Parcel_Petal_Female_CtrlRig.Parcel_Petal_Female_CtrlRig_C}
   }
   Output Pose(Result=@n4) {layerGroup=None}
  ]}
 EventGraph {nodes [
   Event BlueprintUpdateAnimation()
   n1 = GetGameTimeInSeconds()
   n2 = "Sin (Radians)"(A=float * float(A=float * float(A=1, B=5), B=@n1))
   > Set mouthUpperRot(mouthUpperRot:Rotator=MakeRotator(Pitch=float * float(A=@n2, B=4))) -> (Output_Get:Rotator)
   n3 = "Sin (Radians)"(A=float * float(A=float * float(A=1, B=5), B=@n1))
   > Set mouthLowerRot(mouthLowerRot:Rotator=MakeRotator(Pitch=float * float(A=@n3, B=-2))) -> (Output_Get:Rotator)
   TryGetPawnOwner()
  ]}
]
variables [
 mouthLowerRot {type=Rotator}
 mouthUpperRot {type=Rotator}
]
